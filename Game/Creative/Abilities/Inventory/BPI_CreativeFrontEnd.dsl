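; BPI_CreativeFrontEnd is a small Blueprint asset in the Creative inventory abilities folder, named as an interface for the Creative front end.
; It declares one function, GetStoreDeepLink, which takes no inputs and returns a boolean and a string-like value, by name a store deep link.
; The function holds no logic beyond the default false and empty return.
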(fn GetStoreDeepLink
  (return false ""))
