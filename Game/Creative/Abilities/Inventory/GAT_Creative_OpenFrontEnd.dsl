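; Creative gameplay ability template that opens a standalone front-end menu widget when activated from a gameplay event.
; It stores the event instigator, then asynchronously creates the soft-referenced widget for the locally controlled owner where cosmetic events can run.
; The loaded widget gets cursor mode and feeds the empty FrontendLoaded hook; closing it ends the ability.
; Ending the ability commits its cooldown and closes any valid front-end widget.
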(event Ability|EventActivateAbilityFromEvent
  (params (EventData :type Gameplay%20Event%20Data%20Structure%20%28by%20ref%29))
  (bind (_eventtag _instigator) (Utilities|Struct|BreakGameplayEventData EventData))
  (|CallParentFunction EventData)
  (Variables|Default|SetInstigatingObject _instigator)
  (Networking|CanExecuteCosmeticEvents
    (:True
      (bind _abilityowner (Variables|Default|GetAbilityOwner)
        (:then
          (if (Pawn|IsLocallyControlled _abilityowner)
            (bind _userwidget (Class|FortAsyncActionCreateWidgetAsync|CreateWidgetAsync (Variables|Default|GetFrontEndWidgetSoftRef) (Fortnite|GetFirstLocalFortPlayerController true))
              (:OnComplete
                (bind _asfort_standalone_frontend (Utilities|Casting|CastToFortStandaloneFrontend _userwidget))
                (bind _output_get (Variables|Default|SetFrontEndReference _asfort_standalone_frontend))
                (Events|BindEventtoOnMenuClosed _output_get (delegate Custom|Menu%20Closed))
                (HUD|EnterCursorMode (BlueprintContext|GetFortHUDContext) "None" _output_get)
                (CallFunction|FrontendLoaded _output_get)))))
        (:else)))))

(event Custom|Menu%20Closed
  (params (Widget :type Common%20Activatable%20Widget%20Object%20Reference) (Controller :type Player%20Controller%20Object%20Reference))
  (Ability|EndAbility))

(event Ability|EventOnEndAbility
  (params (bWasCancelled :type Boolean))
  (Ability|CommitAbilityCooldown false true)
  (bind _frontendreference (Variables|Default|GetFrontEndReference)
    (:then
      (Class|FortStandaloneFrontend|HandleCloseAction _frontendreference))
    (:else)))

(fn FrontendLoaded
  (params (Loaded%20Widget :type Fort%20Standalone%20Frontend%20Object%20Reference)))
